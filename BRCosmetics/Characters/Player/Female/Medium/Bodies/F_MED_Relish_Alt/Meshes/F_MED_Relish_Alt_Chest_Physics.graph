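# Physics asset for the chest mesh of the female medium Relish Alt player body, covering its left and right dynamic chest straps.
# Wide capsules from pelvis to spine_05 form the torso, joined in a chain with free swing and twist.
# Each strap is a three-body capsule chain whose small master body hangs freely off spine_04.
# Angular limits are tight at the first strap segment and much looser at the tip segment.
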PhysicsAsset /BRCosmetics/Characters/Player/Female/Medium/Bodies/F_MED_Relish_Alt/Meshes/F_MED_Relish_Alt_Chest_Physics
bodies {
 dyn_l_chestStrap_01 [{shapeType=Capsule, length=11.1194, radius=1.1603, shapeName=dyn_l_chestStrap_01_capsule, center=<1.9396 0.0001 0>, rotation=<rot 0 0 90>}]
 dyn_l_chestStrap_02 [{shapeType=Capsule, length=12.7469, radius=2.8136, shapeName=dyn_l_chestStrap_02_capsule, center=<5.8476 0 0>, rotation=<rot 0 0 83.4736>}]
 dyn_l_chestStrap_MSTR [{shapeType=Capsule, length=0.505, radius=0.505, shapeName=dyn_l_chestStrap_MSTR_capsule}]
 dyn_r_chestStrap_01 [{shapeType=Capsule, length=12.1221, radius=1.1941, shapeName=dyn_r_chestStrap_01_capsule, center=<2.5121 0.0001 0>, rotation=<rot 0 0 90>}]
 dyn_r_chestStrap_02 [{
   shapeType=Capsule
   length=10.726
   radius=2.5902
   shapeName=dyn_r_chestStrap_02_capsule
   center=<8.8624 -0.0467 -1.0412>
   rotation=<rot 0 0 92.5731>
  }]
 dyn_r_chestStrap_MSTR [{shapeType=Capsule, length=0.505, radius=0.505, shapeName=dyn_r_chestStrap_MSTR_capsule}]
 pelvis [{shapeType=Capsule, length=8.9224, radius=9.0301, shapeName=pelvis_capsule}]
 spine_01 [{shapeType=Capsule, length=1.4819, radius=9.0301, shapeName=pelvis_capsule}]
 spine_02 [{shapeType=Capsule, length=1.4819, radius=9.0301, shapeName=pelvis_capsule}]
 spine_03 [{shapeType=Capsule, length=1.4819, radius=9.0301, shapeName=pelvis_capsule}]
 spine_04 [{shapeType=Capsule, length=3.9421, radius=9.0301, shapeName=pelvis_capsule}]
 spine_05 [{shapeType=Capsule, length=3.9921, radius=9.0301, shapeName=pelvis_capsule}]
}
constraints [
 {
  bone1Name=spine_01
  bone2Name=pelvis
  swing1LimitDegrees=45
  swing1Motion=Free
  swing2LimitDegrees=45
  swing2Motion=Free
  twistLimitDegrees=45
  twistMotion=Free
 }
 {
  bone1Name=spine_03
  bone2Name=spine_02
  swing1LimitDegrees=45
  swing1Motion=Free
  swing2LimitDegrees=45
  swing2Motion=Free
  twistLimitDegrees=45
  twistMotion=Free
 }
 {
  bone1Name=spine_02
  bone2Name=spine_01
  swing1LimitDegrees=45
  swing1Motion=Free
  swing2LimitDegrees=45
  swing2Motion=Free
  twistLimitDegrees=45
  twistMotion=Free
 }
 {
  bone1Name=spine_04
  bone2Name=spine_03
  swing1LimitDegrees=45
  swing1Motion=Free
  swing2LimitDegrees=45
  swing2Motion=Free
  twistLimitDegrees=45
  twistMotion=Free
 }
 {
  bone1Name=spine_05
  bone2Name=spine_04
  swing1LimitDegrees=45
  swing1Motion=Free
  swing2LimitDegrees=45
  swing2Motion=Free
  twistLimitDegrees=45
  twistMotion=Free
 }
 {
  bone1Name=dyn_r_chestStrap_MSTR
  bone2Name=spine_04
  swing1LimitDegrees=45
  swing1Motion=Free
  swing2LimitDegrees=45
  swing2Motion=Free
  twistLimitDegrees=45
  twistMotion=Free
 }
 {
  bone1Name=dyn_l_chestStrap_MSTR
  bone2Name=spine_04
  swing1LimitDegrees=45
  swing1Motion=Free
  swing2LimitDegrees=45
  swing2Motion=Free
  twistLimitDegrees=45
  twistMotion=Free
 }
 {bone1Name=dyn_r_chestStrap_01, bone2Name=dyn_r_chestStrap_MSTR, swing1LimitDegrees=5, swing2LimitDegrees=5, twistLimitDegrees=5}
 {bone1Name=dyn_r_chestStrap_02, bone2Name=dyn_r_chestStrap_01, swing1LimitDegrees=45, swing2LimitDegrees=45, twistLimitDegrees=45}
 {bone1Name=dyn_l_chestStrap_01, bone2Name=dyn_l_chestStrap_MSTR, swing1LimitDegrees=5, swing2LimitDegrees=5, twistLimitDegrees=5}
 {bone1Name=dyn_l_chestStrap_02, bone2Name=dyn_l_chestStrap_01, swing1LimitDegrees=45, swing2LimitDegrees=45, twistLimitDegrees=45}
]
