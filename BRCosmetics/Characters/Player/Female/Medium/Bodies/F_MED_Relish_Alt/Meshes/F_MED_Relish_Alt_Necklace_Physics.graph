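# Physics asset for the necklace on the F_MED_Relish_Alt female player body, simulating its dangling dynamic bone chain.
# Thin box bodies on the first three necklace bones and a small capsule on the last form a chain hanging from spine_05.
# Tight twist and swing2 limits with a wider swing1 let the chain move mostly about one axis.
# Capsules on spine_04 and spine_05 provide upper-torso shapes, joined by a free-angle constraint.
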PhysicsAsset /BRCosmetics/Characters/Player/Female/Medium/Bodies/F_MED_Relish_Alt/Meshes/F_MED_Relish_Alt_Necklace_Physics
bodies {
 dyn_necklace_01 [{shapeType=Box, extentX=3.6288, extentY=11.5865, extentZ=1.1683, shapeName=dyn_necklace_01_box, center=<1.6666 -0.0948 0.4316>}]
 dyn_necklace_02 [{shapeType=Box, extentX=3.9944, extentY=11.5865, extentZ=1.2646, shapeName=dyn_necklace_02_box, center=<2.1122 0 0.5765>}]
 dyn_necklace_03 [{shapeType=Box, extentX=3.6288, extentY=11.5865, extentZ=1.0691, shapeName=dyn_necklace_03_box, center=<1.6666 0 0.8407>}]
 dyn_necklace_04 [{shapeType=Capsule, length=1.1306, radius=1.0051, shapeName=dyn_necklace_03_capsule, center=<1.4578 0 1.0189>, rotation=<rot 90 0 0>}]
 spine_04 [{shapeType=Capsule, length=9.9157, radius=11.8253, shapeName=spine_05_capsule, center=<4.9013 -3.0548 0>, rotation=<rot 0 9.5675 0>}]
 spine_05 [{shapeType=Capsule, length=13.2618, radius=8.4792, shapeName=spine_05_capsule, center=<2.2381 -1.9645 0>}]
}
constraints [
 {
  bone1Name=spine_05
  bone2Name=spine_04
  swing1LimitDegrees=45
  swing1Motion=Free
  swing2LimitDegrees=45
  swing2Motion=Free
  twistLimitDegrees=45
  twistMotion=Free
 }
 {bone1Name=dyn_necklace_01, bone2Name=spine_05, swing1LimitDegrees=20, swing2LimitDegrees=2, twistLimitDegrees=2}
 {bone1Name=dyn_necklace_02, bone2Name=dyn_necklace_01, swing1LimitDegrees=20, swing2LimitDegrees=2, twistLimitDegrees=2}
 {bone1Name=dyn_necklace_03, bone2Name=dyn_necklace_02, swing1LimitDegrees=20, swing2LimitDegrees=2, twistLimitDegrees=2}
 {bone1Name=dyn_necklace_04, bone2Name=dyn_necklace_03, swing1LimitDegrees=15, swing2LimitDegrees=5, twistLimitDegrees=2}
]
