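# Physics asset for the female Relish Alt face accessory mesh, with dynamic chains for left and right bangs and back strands, likely hair.
# Identical capsules on each strand bone link in chains rooted at a larger head capsule, with tight swing and twist limits.
# The left bang B chain branches at its second bone.
# Small spine and neck capsules link up to the head through constraints with free angular motion.
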PhysicsAsset /BRCosmetics/Characters/Player/Female/Medium/Bodies/F_MED_Relish_Alt/Meshes/Parts/F_MED_Relish_Alt_FaceAcc_FE_Physics
bodies {
 dyn_back_a_01 [{shapeType=Capsule, length=2.3828, radius=1.6172, shapeName=dyn_back_a_01_capsule, center=<2.7304 0 0>, rotation=<rot 90 0 0>}]
 dyn_back_a_02 [{shapeType=Capsule, length=2.3828, radius=1.6172, shapeName=dyn_back_a_02_capsule, center=<2.7304 0 0>, rotation=<rot 90 0 0>}]
 dyn_back_a_03 [{shapeType=Capsule, length=2.3828, radius=1.6172, shapeName=dyn_back_a_03_capsule, center=<2.7304 0 0>, rotation=<rot 90 0 0>}]
 dyn_back_b_01 [{shapeType=Capsule, length=2.3828, radius=1.6172, shapeName=dyn_back_b_01_capsule, center=<2.7304 0 0>, rotation=<rot 90 0 0>}]
 dyn_back_b_02 [{shapeType=Capsule, length=2.3828, radius=1.6172, shapeName=dyn_back_b_02_capsule, center=<2.7304 0 0>, rotation=<rot 90 0 0>}]
 dyn_back_b_03 [{shapeType=Capsule, length=2.3828, radius=1.6172, shapeName=dyn_back_b_03_capsule, center=<2.7304 0 0>, rotation=<rot 90 0 0>}]
 dyn_back_c_01 [{shapeType=Capsule, length=2.3828, radius=1.6172, shapeName=dyn_back_c_01_capsule, center=<2.7304 0 0>, rotation=<rot 90 0 0>}]
 dyn_back_c_02 [{shapeType=Capsule, length=2.3828, radius=1.6172, shapeName=dyn_back_c_02_capsule, center=<2.7304 0 0>, rotation=<rot 90 0 0>}]
 dyn_back_c_03 [{shapeType=Capsule, length=2.3828, radius=1.6172, shapeName=dyn_back_c_03_capsule, center=<2.7304 0 0>, rotation=<rot 90 0 0>}]
 dyn_l_bang_a_01 [{shapeType=Capsule, length=2.3828, radius=1.6172, shapeName=dyn_l_bang_a_01_capsule, center=<2.7304 0 0>, rotation=<rot 90 0 0>}]
 dyn_l_bang_a_02 [{shapeType=Capsule, length=2.3828, radius=1.6172, shapeName=dyn_l_bang_a_02_capsule, center=<2.7304 0 0>, rotation=<rot 90 0 0>}]
 dyn_l_bang_a_03 [{shapeType=Capsule, length=2.3828, radius=1.6172, shapeName=dyn_l_bang_a_03_capsule, center=<2.7304 0 0>, rotation=<rot 90 0 0>}]
 dyn_l_bang_a_04 [{shapeType=Capsule, length=2.3828, radius=1.6172, shapeName=dyn_l_bang_a_04_capsule, center=<2.7304 0 0>, rotation=<rot 90 0 0>}]
 dyn_l_bang_b_01 [{shapeType=Capsule, length=2.3828, radius=1.6172, shapeName=dyn_l_bang_b_01_capsule, center=<2.7304 0 0>, rotation=<rot 90 0 0>}]
 dyn_l_bang_b_02 [{shapeType=Capsule, length=2.3828, radius=1.6172, shapeName=dyn_l_bang_b_02_capsule, center=<2.7304 0 0>, rotation=<rot 90 0 0>}]
 dyn_l_bang_b_03 [{shapeType=Capsule, length=2.3828, radius=1.6172, shapeName=dyn_l_bang_b_03_capsule, center=<2.7304 0 0>, rotation=<rot 90 0 0>}]
 dyn_l_bang_b_04 [{shapeType=Capsule, length=2.3828, radius=1.6172, shapeName=dyn_l_bang_b_04_capsule, center=<2.7304 0 0>, rotation=<rot 90 0 0>}]
 dyn_l_bang_b_05 [{shapeType=Capsule, length=2.3828, radius=1.6172, shapeName=dyn_l_bang_b_05_capsule, center=<2.7304 0 0>, rotation=<rot 90 0 0>}]
 dyn_l_bang_b_06 [{shapeType=Capsule, length=2.3828, radius=1.6172, shapeName=dyn_l_bang_b_06_capsule, center=<2.7304 0 0>, rotation=<rot 90 0 0>}]
 dyn_l_bang_b_07 [{shapeType=Capsule, length=2.3828, radius=1.6172, shapeName=dyn_l_bang_b_07_capsule, center=<2.7304 0 0>, rotation=<rot 90 0 0>}]
 dyn_l_bang_d_01 [{shapeType=Capsule, length=2.3828, radius=1.6172, shapeName=dyn_l_bang_d_01_capsule, center=<2.7304 0 0>, rotation=<rot 90 0 0>}]
 dyn_l_bang_d_02 [{shapeType=Capsule, length=2.3828, radius=1.6172, shapeName=dyn_l_bang_d_02_capsule, center=<2.7304 0 0>, rotation=<rot 90 0 0>}]
 dyn_l_bang_d_03 [{shapeType=Capsule, length=2.3828, radius=1.6172, shapeName=dyn_l_bang_d_03_capsule, center=<2.7304 0 0>, rotation=<rot 90 0 0>}]
 dyn_l_bang_d_04 [{shapeType=Capsule, length=2.3828, radius=1.6172, shapeName=dyn_l_bang_d_04_capsule, center=<2.7304 0 0>, rotation=<rot 90 0 0>}]
 dyn_r_bang_a_01 [{shapeType=Capsule, length=2.3828, radius=1.6172, shapeName=dyn_r_bang_a_01_capsule, center=<2.7304 0 0>, rotation=<rot 90 0 0>}]
 dyn_r_bang_a_02 [{shapeType=Capsule, length=2.3828, radius=1.6172, shapeName=dyn_r_bang_a_02_capsule, center=<2.7304 0 0>, rotation=<rot 90 0 0>}]
 dyn_r_bang_a_03 [{shapeType=Capsule, length=2.3828, radius=1.6172, shapeName=dyn_r_bang_a_03_capsule, center=<2.7304 0 0>, rotation=<rot 90 0 0>}]
 dyn_r_bang_b_01 [{shapeType=Capsule, length=2.3828, radius=1.6172, shapeName=dyn_r_bang_b_01_capsule, center=<2.7304 0 0>, rotation=<rot 90 0 0>}]
 dyn_r_bang_b_02 [{shapeType=Capsule, length=2.3828, radius=1.6172, shapeName=dyn_r_bang_b_02_capsule, center=<2.7304 0 0>, rotation=<rot 90 0 0>}]
 dyn_r_bang_b_03 [{shapeType=Capsule, length=2.3828, radius=1.6172, shapeName=dyn_r_bang_b_03_capsule, center=<2.7304 0 0>, rotation=<rot 90 0 0>}]
 head [{shapeType=Capsule, length=9.7619, radius=7.3468, center=<0 0.8513 0>, rotation=<rot 90 0 -10.6783>}]
 neck_01 [{shapeType=Capsule, length=0.505, radius=0.505, shapeName=neck_01_capsule}]
 neck_02 [{shapeType=Capsule, length=0.505, radius=0.505, shapeName=neck_02_capsule}]
 spine_01 [{shapeType=Capsule, length=0.505, radius=0.505, shapeName=spine_01_capsule}]
 spine_02 [{shapeType=Capsule, length=0.505, radius=0.505, shapeName=spine_02_capsule}]
 spine_03 [{shapeType=Capsule, length=0.505, radius=0.505, shapeName=spine_03_capsule}]
 spine_04 [{shapeType=Capsule, length=0.505, radius=0.505, shapeName=spine_04_capsule}]
 spine_05 [{shapeType=Capsule, length=0.505, radius=0.505, shapeName=spine_05_capsule}]
}
constraints [
 {bone1Name=dyn_l_bang_a_01, bone2Name=head, swing1LimitDegrees=25, swing2LimitDegrees=25, twistLimitDegrees=5}
 {bone1Name=dyn_l_bang_a_02, bone2Name=dyn_l_bang_a_01, swing1LimitDegrees=25, swing2LimitDegrees=25, twistLimitDegrees=5}
 {bone1Name=dyn_l_bang_a_03, bone2Name=dyn_l_bang_a_02, swing1LimitDegrees=25, swing2LimitDegrees=25, twistLimitDegrees=5}
 {bone1Name=dyn_l_bang_a_04, bone2Name=dyn_l_bang_a_03, swing1LimitDegrees=25, swing2LimitDegrees=25, twistLimitDegrees=5}
 {bone1Name=dyn_l_bang_b_01, bone2Name=head, swing1LimitDegrees=25, swing2LimitDegrees=25, twistLimitDegrees=5}
 {bone1Name=dyn_l_bang_b_02, bone2Name=dyn_l_bang_b_01, swing1LimitDegrees=25, swing2LimitDegrees=25, twistLimitDegrees=5}
 {bone1Name=dyn_l_bang_b_03, bone2Name=dyn_l_bang_b_02, swing1LimitDegrees=25, swing2LimitDegrees=25, twistLimitDegrees=5}
 {bone1Name=dyn_l_bang_b_04, bone2Name=dyn_l_bang_b_03, swing1LimitDegrees=25, swing2LimitDegrees=25, twistLimitDegrees=5}
 {bone1Name=dyn_l_bang_b_06, bone2Name=dyn_l_bang_b_02, swing1LimitDegrees=25, swing2LimitDegrees=25, twistLimitDegrees=5}
 {bone1Name=dyn_l_bang_b_07, bone2Name=dyn_l_bang_b_06, swing1LimitDegrees=25, swing2LimitDegrees=25, twistLimitDegrees=5}
 {bone1Name=dyn_l_bang_b_05, bone2Name=dyn_l_bang_b_02, swing1LimitDegrees=25, swing2LimitDegrees=25, twistLimitDegrees=5}
 {bone1Name=dyn_r_bang_b_01, bone2Name=head, swing1LimitDegrees=25, swing2LimitDegrees=25, twistLimitDegrees=5}
 {bone1Name=dyn_r_bang_b_02, bone2Name=dyn_r_bang_b_01, swing1LimitDegrees=25, swing2LimitDegrees=25, twistLimitDegrees=5}
 {bone1Name=dyn_r_bang_b_03, bone2Name=dyn_r_bang_b_02, swing1LimitDegrees=25, swing2LimitDegrees=25, twistLimitDegrees=5}
 {bone1Name=dyn_r_bang_a_01, bone2Name=head, swing1LimitDegrees=25, swing2LimitDegrees=25, twistLimitDegrees=5}
 {bone1Name=dyn_r_bang_a_02, bone2Name=dyn_r_bang_a_01, swing1LimitDegrees=25, swing2LimitDegrees=25, twistLimitDegrees=5}
 {bone1Name=dyn_r_bang_a_03, bone2Name=dyn_r_bang_a_02, swing1LimitDegrees=25, swing2LimitDegrees=25, twistLimitDegrees=5}
 {bone1Name=dyn_l_bang_d_01, bone2Name=head, swing1LimitDegrees=25, swing2LimitDegrees=25, twistLimitDegrees=5}
 {bone1Name=dyn_l_bang_d_02, bone2Name=dyn_l_bang_d_01, swing1LimitDegrees=25, swing2LimitDegrees=25, twistLimitDegrees=5}
 {bone1Name=dyn_l_bang_d_03, bone2Name=dyn_l_bang_d_02, swing1LimitDegrees=25, swing2LimitDegrees=25, twistLimitDegrees=5}
 {bone1Name=dyn_l_bang_d_04, bone2Name=dyn_l_bang_d_03, swing1LimitDegrees=25, swing2LimitDegrees=25, twistLimitDegrees=5}
 {bone1Name=dyn_back_a_01, bone2Name=head, swing1LimitDegrees=25, swing2LimitDegrees=25, twistLimitDegrees=5}
 {bone1Name=dyn_back_a_02, bone2Name=dyn_back_a_01, swing1LimitDegrees=25, swing2LimitDegrees=25, twistLimitDegrees=5}
 {bone1Name=dyn_back_a_03, bone2Name=dyn_back_a_02, swing1LimitDegrees=25, swing2LimitDegrees=25, twistLimitDegrees=5}
 {bone1Name=dyn_back_b_01, bone2Name=head, swing1LimitDegrees=25, swing2LimitDegrees=25, twistLimitDegrees=5}
 {bone1Name=dyn_back_b_02, bone2Name=dyn_back_b_01, swing1LimitDegrees=25, swing2LimitDegrees=25, twistLimitDegrees=5}
 {bone1Name=dyn_back_b_03, bone2Name=dyn_back_b_02, swing1LimitDegrees=25, swing2LimitDegrees=25, twistLimitDegrees=5}
 {bone1Name=dyn_back_c_01, bone2Name=head, swing1LimitDegrees=25, swing2LimitDegrees=25, twistLimitDegrees=5}
 {bone1Name=dyn_back_c_02, bone2Name=dyn_back_c_01, swing1LimitDegrees=25, swing2LimitDegrees=25, twistLimitDegrees=5}
 {bone1Name=dyn_back_c_03, bone2Name=dyn_back_c_02, swing1LimitDegrees=25, swing2LimitDegrees=25, twistLimitDegrees=5}
 {
  bone1Name=head
  bone2Name=neck_02
  swing1LimitDegrees=45
  swing1Motion=Free
  swing2LimitDegrees=45
  swing2Motion=Free
  twistLimitDegrees=45
  twistMotion=Free
 }
 {
  bone1Name=spine_02
  bone2Name=spine_01
  swing1LimitDegrees=45
  swing1Motion=Free
  swing2LimitDegrees=45
  swing2Motion=Free
  twistLimitDegrees=45
  twistMotion=Free
 }
 {
  bone1Name=spine_03
  bone2Name=spine_02
  swing1LimitDegrees=45
  swing1Motion=Free
  swing2LimitDegrees=45
  swing2Motion=Free
  twistLimitDegrees=45
  twistMotion=Free
 }
 {
  bone1Name=spine_04
  bone2Name=spine_03
  swing1LimitDegrees=45
  swing1Motion=Free
  swing2LimitDegrees=45
  swing2Motion=Free
  twistLimitDegrees=45
  twistMotion=Free
 }
 {
  bone1Name=spine_05
  bone2Name=spine_04
  swing1LimitDegrees=45
  swing1Motion=Free
  swing2LimitDegrees=45
  swing2Motion=Free
  twistLimitDegrees=45
  twistMotion=Free
 }
 {
  bone1Name=neck_01
  bone2Name=spine_05
  swing1LimitDegrees=45
  swing1Motion=Free
  swing2LimitDegrees=45
  swing2Motion=Free
  twistLimitDegrees=45
  twistMotion=Free
 }
 {
  bone1Name=neck_02
  bone2Name=neck_01
  swing1LimitDegrees=45
  swing1Motion=Free
  swing2LimitDegrees=45
  swing2Motion=Free
  twistLimitDegrees=45
  twistMotion=Free
 }
]
