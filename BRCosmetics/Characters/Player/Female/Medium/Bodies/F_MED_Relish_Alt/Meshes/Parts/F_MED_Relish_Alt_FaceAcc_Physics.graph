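# Physics asset for the face accessory mesh of the F_MED_Relish_Alt female player cosmetic, covering three dynamic bone chains on the back of the head.
# Each chain has three capsules that widen toward the tip, constrained link to link from the head with asymmetric swing and narrow twist limits.
# Capsules on the head and two neck bones complete the body set, joined by constraints with free swing and twist.
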PhysicsAsset /BRCosmetics/Characters/Player/Female/Medium/Bodies/F_MED_Relish_Alt/Meshes/Parts/F_MED_Relish_Alt_FaceAcc_Physics
bodies {
 dyn_back_a_01 [{shapeType=Capsule, length=2.6973, radius=1.3027, shapeName=dyn_back_a_01_capsule, center=<2.7817 -0.0001 0>, rotation=<rot 0 0 90>}]
 dyn_back_a_02 [{shapeType=Capsule, length=2.3944, radius=1.6056, shapeName=dyn_back_a_02_capsule, center=<2.7817 -0.0001 0>, rotation=<rot 0 0 90>}]
 dyn_back_a_03 [{shapeType=Capsule, length=2, radius=2, shapeName=dyn_back_a_03_capsule, center=<2.7817 -0.0001 0>, rotation=<rot 0 0 90>}]
 dyn_back_b_01 [{shapeType=Capsule, length=2.6973, radius=1.3027, shapeName=dyn_back_b_01_capsule, center=<2.7817 -0.0001 0>, rotation=<rot 0 0 90>}]
 dyn_back_b_02 [{shapeType=Capsule, length=2.3944, radius=1.6056, shapeName=dyn_back_b_02_capsule, center=<2.7817 -0.0001 0>, rotation=<rot 0 0 90>}]
 dyn_back_b_03 [{shapeType=Capsule, length=2, radius=2, shapeName=dyn_back_b_03_capsule, center=<2.7817 -0.0001 0>, rotation=<rot 0 0 90>}]
 dyn_back_c_01 [{shapeType=Capsule, length=2.6973, radius=1.3027, shapeName=dyn_back_c_01_capsule, center=<2.7817 -0.0001 0>, rotation=<rot 0 0 90>}]
 dyn_back_c_02 [{shapeType=Capsule, length=2.3944, radius=1.6056, shapeName=dyn_back_c_02_capsule, center=<2.7817 -0.0001 0>, rotation=<rot 0 0 90>}]
 dyn_back_c_03 [{shapeType=Capsule, length=2, radius=2, shapeName=dyn_back_c_03_capsule, center=<2.7817 -0.0001 0>, rotation=<rot 0 0 90>}]
 head [{
   shapeType=Capsule
   length=3.8259
   radius=6.4518
   shapeName=head_capsule
   center=<3.226 1.0005 -0.4726>
   rotation=<rot -16.2535 -63.697 95.2218>
  }]
 neck_01 [{shapeType=Capsule, length=7.92, radius=3.678, shapeName=neck_01_capsule}]
 neck_02 [{shapeType=Capsule, length=7.92, radius=3.678, shapeName=neck_02_capsule}]
}
constraints [
 {
  bone1Name=head
  bone2Name=neck_02
  swing1LimitDegrees=45
  swing1Motion=Free
  swing2LimitDegrees=45
  swing2Motion=Free
  twistLimitDegrees=45
  twistMotion=Free
 }
 {
  bone1Name=neck_02
  bone2Name=neck_01
  swing1LimitDegrees=45
  swing1Motion=Free
  swing2LimitDegrees=45
  swing2Motion=Free
  twistLimitDegrees=45
  twistMotion=Free
 }
 {bone1Name=dyn_back_a_01, bone2Name=head, swing1LimitDegrees=30, swing2LimitDegrees=45, twistLimitDegrees=15}
 {bone1Name=dyn_back_a_02, bone2Name=dyn_back_a_01, swing1LimitDegrees=30, swing2LimitDegrees=45, twistLimitDegrees=15}
 {bone1Name=dyn_back_a_03, bone2Name=dyn_back_a_02, swing1LimitDegrees=30, swing2LimitDegrees=45, twistLimitDegrees=15}
 {bone1Name=dyn_back_b_01, bone2Name=head, swing1LimitDegrees=30, swing2LimitDegrees=45, twistLimitDegrees=15}
 {bone1Name=dyn_back_b_02, bone2Name=dyn_back_b_01, swing1LimitDegrees=30, swing2LimitDegrees=45, twistLimitDegrees=15}
 {bone1Name=dyn_back_b_03, bone2Name=dyn_back_b_02, swing1LimitDegrees=30, swing2LimitDegrees=45, twistLimitDegrees=15}
 {bone1Name=dyn_back_c_01, bone2Name=head, swing1LimitDegrees=30, swing2LimitDegrees=45, twistLimitDegrees=15}
 {bone1Name=dyn_back_c_02, bone2Name=dyn_back_c_01, swing1LimitDegrees=30, swing2LimitDegrees=45, twistLimitDegrees=15}
 {bone1Name=dyn_back_c_03, bone2Name=dyn_back_c_02, swing1LimitDegrees=30, swing2LimitDegrees=45, twistLimitDegrees=15}
]
